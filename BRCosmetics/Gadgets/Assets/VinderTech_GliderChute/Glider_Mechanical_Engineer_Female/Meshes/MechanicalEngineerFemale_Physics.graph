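# Physics asset for the female Mechanical Engineer glider mesh, among the VinderTech GliderChute cosmetic gadget assets.
# It defines a single body on the wrench_body bone, using one large, nearly spherical capsule offset from the bone origin.
# It has no constraints, so the asset provides one coarse collision shape rather than an articulated ragdoll.
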PhysicsAsset /BRCosmetics/Gadgets/Assets/VinderTech_GliderChute/Glider_Mechanical_Engineer_Female/Meshes/MechanicalEngineerFemale_Physics
bodies {wrench_body [{shapeType=Capsule, length=8.027, radius=127.7244, center=<0.4258 -11.1893 69.0808>, rotation=<rot 0.6042 -90 -0.0202>}]}
constraints []
